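# Translucent UI material for the primary quick weapon element in the inventory HUD, with highlight and empty states.
# One mask texture drives it: green blends falloff into hotspot color, and blue times HighlightStrength blends in the highlight color.
# Opacity is red plus that highlight term.
# IsEmpty blends color toward Empty_Color and opacity toward its alpha times the blue mask.
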Material /Game/Athena/HUD/Inventory/Art/M_UI_QuickWeapon_Primary
graph {
 properties {blendMode=BLEND_Translucent, materialDomain=MD_UI, pixelDepthOffsetMode=PDOM_Legacy, usedShadingModels=Default Lit}
 nodes [
  n1 = VectorParameter() {parameterName=BG_Falloff_Color, defaultValue=<rgba 0 0.0046 0.7083 1>}
  n2 = VectorParameter() {parameterName=BG_Hotspot_Color, defaultValue=<rgba 0 0.7472 1 1>}
  n3 = TextureSampleParameter2D() {parameterName=Masks, texture=&/T_UI_QuickWeapon_Primary_256x}
  n4 = Multiply(A=@n3.B, B=ScalarParameter() {parameterName=HighlightStrength})
  n5 = VectorParameter() {parameterName=Empty_Color, defaultValue=<rgba 0.5 0.5 0.5 0.25>}
  n6 = ScalarParameter() {parameterName=IsEmpty}
  LinearInterpolate(
   A=LinearInterpolate(
    A=LinearInterpolate(A=@n1.RGB, B=@n2.RGB, Alpha=@n3.G)
    B=VectorParameter() {parameterName=Highlight_Color, defaultValue=<rgba 1 1 1 1>}.RGB
    Alpha=@n4
   )
   B=@n5.RGB
   Alpha=@n6
  )
  LinearInterpolate(A=Add(A=@n3.R, B=@n4), B=Multiply(A=@n5.A, B=@n3.B), Alpha=@n6)
 ]
}
